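# Sound cue for a pinball flipper flip on a Creative toy.
# A Random node picks one of four equally weighted flip recordings, and a Modulator varies pitch slightly around normal and volume slightly upward on each play.
# The cue plays at reduced volume in the Creative gadgets sound class, spatialized with the medium toy attenuation asset.
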SoundCue /Game/Sounds/Creative/Toys/Pinball/Pinball_Flipper_Flip_Cue
root=Modulator_0
nodes [
 {id=Modulator_0, children [Random_0], properties {pitchMax=1.1, pitchMin=0.9, volumeMax=1.1, volumeMin=1}}
 {id=Random_0, children [WavePlayer_1,WavePlayer_2,WavePlayer_3,WavePlayer_6], properties {weights [1,1,1,1]}}
 {id=WavePlayer_1, properties {soundWaveAssetPtr=&/Pinball_Flipper_Flip_01}}
 {id=WavePlayer_2, properties {soundWaveAssetPtr=&/Pinball_Flipper_Flip_02}}
 {id=WavePlayer_3, properties {soundWaveAssetPtr=&/Pinball_Flipper_Flip_03}}
 {id=WavePlayer_6, properties {soundWaveAssetPtr=&/Pinball_Flipper_Flip_04}}
]
properties {
 attenuationSettings=/Game/Sounds/Creative/Attenuations/Fort_Creative_MediumToy_Att
 maxDistance=4500
 soundClassObject=&Fort_Gadgets_Creative
 volumeMultiplier=0.4
}
