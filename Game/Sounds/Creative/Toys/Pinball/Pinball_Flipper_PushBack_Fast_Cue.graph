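# Sound cue for the fast push-back sound of a pinball flipper toy in Creative.
# A random node picks one of four equally weighted wave variations and feeds a modulator that slightly varies pitch and volume.
# It plays with medium-toy attenuation in the Creative gadgets sound class.
# Concurrency allows one instance per owner and blocks new plays rather than cutting the current one.
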SoundCue /Game/Sounds/Creative/Toys/Pinball/Pinball_Flipper_PushBack_Fast_Cue
root=Modulator_0
nodes [
 {id=Modulator_0, children [Random_1], properties {pitchMax=1.1, pitchMin=0.9, volumeMax=1.1, volumeMin=1}}
 {id=Random_1, children [WavePlayer_0,WavePlayer_4,WavePlayer_5,WavePlayer_7], properties {weights [1,1,1,1]}}
 {id=WavePlayer_0, properties {soundWaveAssetPtr=&/Pinball_Flipper_PushBack_Fast_01}}
 {id=WavePlayer_4, properties {soundWaveAssetPtr=&/Pinball_Flipper_PushBack_Fast_02}}
 {id=WavePlayer_5, properties {soundWaveAssetPtr=&/Pinball_Flipper_PushBack_Fast_03}}
 {id=WavePlayer_7, properties {soundWaveAssetPtr=&/Pinball_Flipper_PushBack_Fast_04}}
]
properties {
 attenuationSettings=/Game/Sounds/Creative/Attenuations/Fort_Creative_MediumToy_Att
 maxDistance=4500
 soundClassObject=&Fort_Gadgets_Creative
 volumeMultiplier=0.4
 concurrencyOverrides {bLimitToOwner=true, maxCount=1, resolutionRule=PreventNew}
}
