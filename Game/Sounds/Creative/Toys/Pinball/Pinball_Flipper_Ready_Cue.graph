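# Sound cue for the Creative pinball toy, named as the flipper ready sound.
# It mixes the jump pad placed wave with a quieter scoreboard score-reset wave, both reused from other assets.
# An enveloper over the mix ramps pitch upward and holds volume briefly before fading it out within half a second.
# The cue plays pitched up at reduced volume, using small-toy attenuation and the Creative gadgets sound class.
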SoundCue /Game/Sounds/Creative/Toys/Pinball/Pinball_Flipper_Ready_Cue
root=Enveloper_0
nodes [
 {
  id=Enveloper_0
  children [Mixer_0]
  properties {pitchCurve {editorCurveData {keys [{time=0}, <at 0.5 0.5>]}}, volumeCurve {editorCurveData {keys [<at 0 1>,<at 0.3 1>,<at 0.5 0>]}}}
 }
 {id=Mixer_0, children [WavePlayer_0,WavePlayer_1], properties {inputVolume [1,0.7]}}
 {id=WavePlayer_0, properties {soundWaveAssetPtr=/Game/Sounds/Fort_Traps/JumpPad/JumpPad_Placed_001}}
 {id=WavePlayer_1, properties {soundWaveAssetPtr=/Game/Sounds/Creative/Toys/Scoreboard/Player_Scoreboard_ScoreReset}}
]
properties {
 attenuationSettings=/Game/Sounds/Creative/Attenuations/Fort_Creative_SmallToy_Att
 maxDistance=3000
 pitchMultiplier=2
 soundClassObject=&Fort_Gadgets_Creative
 volumeMultiplier=0.3
}
